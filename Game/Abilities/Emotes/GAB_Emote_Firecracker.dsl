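; Gameplay ability for the Firecracker emote that blocks the player's look input while it is active.
; On activation it runs the parent logic, sets bIgnoreLookInput to true and applies it to the player pawn's controller.
; On end, cancelled or not, it runs the parent logic, then applies false to release look input.
; The emote's own behavior is not defined here beyond these parent calls.
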(event Ability|EventActivateAbility
  (|CallParentFunction)
  (Variables|Default|SetbIgnoreLookInput true)
  (Input|SetIgnoreLookInput (Pawn|GetController (Variables|Default|GetPlayerPawn)) (Variables|Default|GetbIgnoreLookInput)))

(event Ability|EventOnEndAbility
  (params (bWasCancelled :type Boolean))
  (|CallParentFunction bWasCancelled)
  (Variables|Default|SetbIgnoreLookInput false)
  (Input|SetIgnoreLookInput (Pawn|GetController (Variables|Default|GetPlayerPawn)) (Variables|Default|GetbIgnoreLookInput)))
