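; Gameplay ability for group emotes: the player plays a request montage and waits for another player to join in a synced emote.
; An interaction starts the sync unless either pawn is movement-blocked or the joiner is DBNO.
; StartGroupSyncEmote makes the owner leader, sets look targets and sync from the dance item, plays both group emotes and ends when the partner stops.
; Listener unbinding waits until sync starts, letting derived classes abort and keep listening.
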(event Ability|EventActivateAbility
  ;[0.01357 1 0] Activate Ability
  (|CallParentFunction)
  (bind _playerpawn (Variables|Default|GetPlayerPawn)
    (:then
      (Variables|/Script/FortniteGame.FortPlayerPawn|SetbIsWaitingForEmoteInteraction true _playerpawn) ; GAB_Emote_Generic::ActivateAbility sets IsWaitingForEmoteInteraction to false.\nWe need to make sure it gets set to true even if we do NOT play the montage (for example: if we already started the montage and the emote was requested again)
    )
    (:else)))

(event EventPlayInitialEmoteMontage
  (bind _montageinfo (Variables|Animation|GetMontageInfo))
  (bind _montagetoplay (Utilities|Struct|BreakFortGameplayAbilityMontageInfo _montageinfo))
  (bind _returnvalue (CallFunction|GetCurrentEmoteItemDefinition))
  (bind _playerpawn (Variables|Default|GetPlayerPawn))
  ;[0.6778 0.1161 1] Activate the montage
  (Utilities|IsValid _returnvalue
    (:Is%20Valid
      (if (Emote|ShouldHolsterWeapon _returnvalue _playerpawn)
        (Variables|/Script/FortniteGame.FortPlayerPawn|SetbIsWaitingForEmoteInteraction true _playerpawn)
        (Weapon|SetHolsterWeapon (Variables|Default|GetPlayerPawn) 0 true false (Variables|Default|GetEmoteHolsterId))
        (else
          _))
      (if (not (Variables|Default|GetbAbilityStopped))
        (Ability|Tasks|PlayMontageInfoAndWaitForTrigger "None" _montageinfo
          (:then
            (if (Variables|Default|GetbAbilityStopped)
              (CallFunction|ForceStopMontage)
              (else
                (Utilities|FlowControl|Sequence
                  (:then_0
                    ;[1 1 1] Show/hide reticle if requested
                    (if (Variables|Reticle|GetHideReticle)
                      (CallFunction|SetReticleVisibility true)))
                  (:then_1
                    (Game|Player|BindEventtoOnEmoteInteracted (Variables|Default|GetPlayerPawn) (delegate Custom|EmoteInteracted)))))))
          (:Cancelled
            (Utilities|FlowControl|Sequence
              (:then_0
                (Weapon|SetHolsterWeapon (Variables|Default|GetPlayerPawn) 0 false false (Variables|Default|GetEmoteHolsterId)))
              (:then_1
                (Utilities|IsValid _montagetoplay
                  (:Is%20Valid
                    (Utilities|FlowControl|Delay (Montage|GetDefaultBlendOutTime _montagetoplay)))
                  (:Is%20Not%20Valid
                    _))
                (Ability|EndAbility)))))))
    (:Is%20Not%20Valid
      _)))

(event Custom|EmoteInteracted
  (params (Instigator :type Fort%20Player%20Pawn%20Object%20Reference))
  (bind _playerpawn (Variables|Default|GetPlayerPawn))
  ;[0.1692 1 0.5832] EmoteInteracted
  (Utilities|IsValid _playerpawn
    (:Is%20Valid
      (if (not (Math|Boolean|ORBoolean (Class|FortPlayerPawn|IsEmoteBlockedbyMovement Instigator) (Class|FortPlayerPawn|IsEmoteBlockedbyMovement _playerpawn) (GameplayTags|HasMatchingGameplayTag Instigator "(TagName=\"Gameplay.Status.DBNO\")")))
        (CallFunction|StartGroupSyncEmote Instigator) ; We purposefully skip unbinding EmoteInteracted until StartGroupSyncEmote runs.\nThis allows derived classes to abort the synced emote and continue to listen for EmoteInteracted.
      ))))

(event Ability|EventOnEndAbility
  (params (bWasCancelled :type Boolean))
  (bind _playerpawn (Variables|Default|GetPlayerPawn))
  ;[1 0.5373 0] OnEndAbility
  (|CallParentFunction bWasCancelled)
  (Utilities|IsValid _playerpawn
    (:Is%20Valid
      (Game|Player|UnbindEventfromOnEmoteInteracted _playerpawn (delegate Custom|EmoteInteracted)))))

(event Custom|StartGroupSyncEmote
  (params (Instigator%20Pawn :type Fort%20Player%20Pawn%20Object%20Reference))
  (bind _playerpawn (Variables|Default|GetPlayerPawn))
  ;[0.2543 0.3914 1] Start Group Sync Emote
  (bind _playerpawn_2 (Variables|Default|GetPlayerPawn)
    (:then
      (Game|Player|UnbindEventfromOnEmoteInteracted _playerpawn_2 (EventDispatchers|CreateEvent :!event Custom|EmoteInteracted)) ; We are now starting the synced part of the emote\nso we want to stop listening for other players to interact with our emote.
      (Variables|/Script/FortniteGame.FortPlayerPawn|SetbIsWaitingForEmoteInteraction false Instigator%20Pawn)
      (bind _asathena_dance_item_definition (Utilities|Casting|CastToAthenaDanceItemDefinition (CallFunction|GetCurrentEmoteItemDefinition)))
      (Utilities|FlowControl|Sequence
        (:then_0
          ;;[0.3632 0.9602 1] Setup Group Emote variables
          (Variables|/Script/FortniteGame.FortPlayerPawn|SetGroupEmoteLookTarget Instigator%20Pawn _playerpawn)
          (Variables|/Script/FortniteGame.FortPlayerPawn|SetbShouldSyncAnimationWithEmoteLeader (Variables|/Script/FortniteGame.AthenaDanceItemDefinition|GetbGroupAnimationSync _asathena_dance_item_definition) Instigator%20Pawn)
          (Variables|/Script/FortniteGame.FortPlayerPawn|SetbShouldJitterAnimationSyncWithEmoteLeader false Instigator%20Pawn)
          (Variables|/Script/FortniteGame.FortPlayerPawn|SetGroupEmoteLookTarget _playerpawn Instigator%20Pawn)
          (Variables|/Script/FortniteGame.FortPlayerPawn|SetbIsEmoteLeader true _playerpawn)
          (Variables|/Script/FortniteGame.FortPlayerPawn|SetbLockGroupEmoteLeaderRotation (Variables|/Script/FortniteGame.AthenaDanceItemDefinition|GetbLockGroupEmoteLeaderRotation _asathena_dance_item_definition) _playerpawn)
          (Variables|/Script/FortniteGame.FortPlayerPawn|SetbShouldSyncAnimationWithEmoteLeader (Variables|/Script/FortniteGame.AthenaDanceItemDefinition|GetbGroupAnimationSync _asathena_dance_item_definition) _playerpawn)
          (if (Variables|/Script/FortniteGame.AthenaDanceItemDefinition|GetbLockGroupEmoteLeaderRotation _asathena_dance_item_definition)
            (Variables|/Script/FortniteGame.FortPlayerPawn|SetGroupEmoteLookTargetRotationLeader (Transformation|GetActorRotation _playerpawn) _playerpawn)
            (Variables|/Script/FortniteGame.FortPlayerPawn|SetGroupEmoteLeaderRotationYawOffset (Variables|/Script/FortniteGame.AthenaDanceItemDefinition|GetGroupEmoteLeaderRotationYawOffset _asathena_dance_item_definition) _playerpawn)))
        (:then_1
          ;;[0.061 0.15 0.0593] Handle Event when follower Leaves
          (EventDispatchers|BindEventtoOnGroupEmotePartnerStopped (Variables|Default|GetPlayerPawn) (delegate Custom|CustomEvent)))
        (:then_2
          ;;[0.581 0.08239 1] Make Pawns play emotes
          (bind _returnvalue (Ability|GetGroupEmote _asathena_dance_item_definition (Helpers|ValidateOwnershipforGroupEmote Instigator%20Pawn _asathena_dance_item_definition) true))
          (Class|FortPlayerPawn|PlayGroupEmote (Variables|Default|GetPlayerPawn) _returnvalue)
          (bind _returnvalue_2 (Ability|GetGroupEmote _asathena_dance_item_definition (Helpers|ValidateOwnershipforGroupEmote Instigator%20Pawn _asathena_dance_item_definition)))
          (Class|FortPlayerPawn|PlayGroupEmote Instigator%20Pawn _returnvalue_2))))
    (:else)))

(event Custom|CustomEvent
  (params (StoppedPawn :type Fort%20Player%20Pawn%20Object%20Reference))
  ;[0.2543 0.3914 1] Start Group Sync Emote
  ;;[0.061 0.15 0.0593] Handle Event when follower Leaves
  (Ability|EndAbility)
  (EventDispatchers|UnbindEventfromOnGroupEmotePartnerStopped (Variables|Default|GetPlayerPawn) (delegate Custom|CustomEvent)))
